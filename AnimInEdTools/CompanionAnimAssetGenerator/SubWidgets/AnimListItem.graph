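# Editor utility list-row widget for the Companion Anim Asset Generator, showing one animation name inside an include checkbox on a dark bordered header.
# ListItemSetup writes RowName into the label and sets AC Ref, an AnimatableCreature reference, from a same-named getter.
# ColumnName, the direction-widget array, IsLocomotion and LastGoodEntry are unused here, and the event graph holds only empty events.
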EditorUtilityWidgetBlueprint /AnimInEdTools/CompanionAnimAssetGenerator/SubWidgets/AnimListItem
graphs [
 EventGraph {nodes [
   Event PreConstruct()
   Event Construct()
   Event Tick()
  ]}
 ListItemSetup {nodes [
   n1 = ListItemSetup!() -> (then, RowName:String, AC Ref:Animatable Creature Object Reference, ColumnName:String) {is=title K2Node_FunctionEntry}
   > Set Text(Text="To Text (String)"(InString=@n1), self=Get AnimNameTxt() -> (AnimNameTxt:Text Object Reference))
   > Set AC Ref(AC Ref:Animatable Creature Object Reference=Get AC Ref() -> (AC Ref:Animatable Creature Object Reference)) -> (Output_Get:Animatable Creature Object Reference)
  ]}
]
variables [
 AC Ref {type=AnimatableCreature_C, default=null}
 DirectionWidgetInstances {type="array<Object>", default []}
 IsLocomotion {type=boolean}
 LastGoodEntry {type=string, default=""}
]
widgets {
 info {rootWidgetClass=&VerticalBox, widgetCount=7}
 widgets [
  VerticalBox_0 {widgetClassPath=&VerticalBox, widget=VerticalBox_0}
  HeaderBorder {
   widgetClassPath=&Border
   parent=VerticalBox_0
   widget=HeaderBorder
   properties {brushColor=<rgba 0.0417 0.0417 0.0417 1>, padding=<ltrb 2 2 2 2>}
   slot {class=VerticalBoxSlot, properties {padding=<ltrb 0 0 0 1>}}
  }
  VerticalBox_1 {
   widgetClassPath=&VerticalBox
   parent=HeaderBorder
   widget=VerticalBox_1
   slot {class=BorderSlot, properties {padding=<ltrb 2 2 2 2>}}
  }
  HorizontalBox_1 {widgetClassPath=&HorizontalBox, parent=VerticalBox_1, widget=HorizontalBox_1, slot {class=VerticalBoxSlot}}
  HorizontalBox_202 {widgetClassPath=&HorizontalBox, parent=HorizontalBox_1, widget=HorizontalBox_202, slot {class=HorizontalBoxSlot}}
  IncludeCheck {
   widgetClassPath=&EditorUtilityCheckBox
   bIsVariable=true
   parent=HorizontalBox_202
   widget=IncludeCheck
   slot {class=HorizontalBoxSlot, properties {padding=<ltrb 0 0 5 0>, size {sizeRule=Automatic, value=0.3}}}
  }
  AnimNameTxt {
   widgetClassPath=&TextBlock
   bIsVariable=true
   parent=IncludeCheck
   widget=AnimNameTxt
   properties {
    text=AnimName
    font {
     bForceMonospaced=false
     bMaterialIsStencil=false
     fontMaterial=null
     fontObject=&Roboto
     letterSpacing=0
     monospacedWidth=1
     size=9
     skewAmount=0
     typefaceFontName=Bold
     outlineSettings {
      bApplyOutlineToDropShadows=false
      bMiteredCorners=false
      bSeparateFillAlpha=false
      outlineMaterial=null
      outlineSize=0
      outlineColor=<rgba 0 0 0 1>
     }
    }
   }
   slot {class=PanelSlot}
  }
 ]
}
